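# Physics asset for the Pickaxe_Football_Referee two-handed melee cosmetic, giving its mesh a jointed flag top.
# Three capsule bodies cover the handle, a tiny root bone and the dyn_flag_top bone.
# The handle and root are rotationally locked together.
# The flag connects to the handle through a joint whose swing and twist are not locked, so it can rotate relative to the handle.
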PhysicsAsset /BRCosmetics/Weapons/FORT_Melee/Pickaxe_Football_Referee/Meshes/Pickaxe_Football_Referee_PhysicsAsset
bodies {
 Melee_TwoHanded_Handle [{shapeType=Capsule, length=60, radius=5}]
 b_AR_Root [{shapeType=Capsule, length=0.505, radius=0.505, rotation=<rot 0 -90 0>}]
 dyn_flag_top [{shapeType=Capsule, length=15, radius=12, center=<0.25 0 0.5>, rotation=<rot 0 0 90>}]
}
constraints [
 {
  bone1Name=Melee_TwoHanded_Handle
  bone2Name=b_AR_Root
  swing1LimitDegrees=45
  swing1Motion=Locked
  swing2LimitDegrees=45
  swing2Motion=Locked
  twistLimitDegrees=45
  twistMotion=Locked
 }
 {bone1Name=dyn_flag_top, bone2Name=Melee_TwoHanded_Handle, swing1LimitDegrees=90, swing2LimitDegrees=15, twistLimitDegrees=90}
]
